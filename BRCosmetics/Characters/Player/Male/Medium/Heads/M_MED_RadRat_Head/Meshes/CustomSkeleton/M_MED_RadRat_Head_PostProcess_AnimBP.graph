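# Post-process AnimBP for the RadRat male head that adds secondary hair motion.
# A Control Rig feeds weighted multi-bone constraints blending six base_hair bones across head, neck, spine and clavicle, then AnimDynamics angular springs simulate fourteen dyn_hair bones.
# A second constraint chain blending base_hair with dyn_hair bones is disconnected from output.
# Each update sets simulation space and alpha variables by front-end state, but no node visibly reads them.
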AnimBlueprint /BRCosmetics/Characters/Player/Male/Medium/Heads/M_MED_RadRat_Head/Meshes/CustomSkeleton/M_MED_RadRat_Head_PostProcess_AnimBP
target_skeleton=&/Fortnite_Base_Head_RadRat_Export_Skeleton
graphs [
 AnimGraph {nodes [
   n1 = "Constraint - base_hair_02 (PP)"!(ComponentPose, Alpha=1, ConstraintWeights_0=0.9071, ConstraintWeights_1:float=0.0929) -> (Pose) {
    boneToModify {boneName=base_hair_02}
    constraintSetup [{targetBone {boneName=head}}, {targetBone {boneName=dyn_hair_02}}]
    constraintWeights [1,1]
   }
   n2 = "Constraint - base_hair_03 (PPPPPPPP)"!(
    ComponentPose=@n1
    Alpha=1
    ConstraintWeights_0=0.259
    ConstraintWeights_1:float=0.2516
    ConstraintWeights_2:float=0.2428
    ConstraintWeights_3:float=0.0827
    ConstraintWeights_4:float=0.0743
    ConstraintWeights_5:float=0.0508
    ConstraintWeights_6:float=0.0297
    ConstraintWeights_7:float=0.0091
   ) -> (Pose) {
    boneToModify {boneName=base_hair_03}
    constraintSetup [
     {targetBone {boneName=spine_05}}
     {targetBone {boneName=neck_02}}
     {targetBone {boneName=neck_01}}
     {targetBone {boneName=head}}
     {targetBone {boneName=dyn_hair_03}}
     {targetBone {boneName=dyn_hair_l_adjust}}
     {targetBone {boneName=spine_03}}
     {targetBone {boneName=spine_04}}
    ]
    constraintWeights [1,1,1,1,1,1,1,1]
   }
   n3 = "Constraint - base_hair_04 (PPPPPPP)"!(
    ComponentPose=@n2
    Alpha=1
    ConstraintWeights_0=0.3557
    ConstraintWeights_1:float=0.2523
    ConstraintWeights_2:float=0.1988
    ConstraintWeights_3:float=0.1528
    ConstraintWeights_4:float=0.0277
    ConstraintWeights_5:float=0.0102
    ConstraintWeights_6:float=0.0024
   ) -> (Pose) {
    boneToModify {boneName=base_hair_04}
    constraintSetup [
     {targetBone {boneName=neck_02}}
     {targetBone {boneName=neck_01}}
     {targetBone {boneName=dyn_hair_04}}
     {targetBone {boneName=head}}
     {targetBone {boneName=spine_03}}
     {targetBone {boneName=spine_04}}
     {targetBone {boneName=base_hair_09}}
    ]
    constraintWeights [1,1,1,1,1,1,1]
   }
   n4 = "Constraint - base_hair_05 (PPPPPPPP)"!(
    ComponentPose=@n3
    Alpha=1
    ConstraintWeights_0=0.2484
    ConstraintWeights_1:float=0.2049
    ConstraintWeights_2:float=0.1985
    ConstraintWeights_3:float=0.1824
    ConstraintWeights_4:float=0.133
    ConstraintWeights_5:float=0.0226
    ConstraintWeights_6:float=0.0083
    ConstraintWeights_7:float=0.002
   ) -> (Pose) {
    boneToModify {boneName=base_hair_05}
    constraintSetup [
     {targetBone {boneName=neck_01}}
     {targetBone {boneName=neck_02}}
     {targetBone {boneName=dyn_hair_05}}
     {targetBone {boneName=spine_05}}
     {targetBone {boneName=head}}
     {targetBone {boneName=spine_03}}
     {targetBone {boneName=spine_04}}
     {targetBone {boneName=dyn_hair_l_adjust}}
    ]
    constraintWeights [1,1,1,1,1,1,1,1]
   }
   n5 = "Constraint - base_hair_06 (PPPPPPP)"!(
    ComponentPose=@n4
    Alpha=1
    ConstraintWeights_0=0.3449
    ConstraintWeights_1:float=0.2761
    ConstraintWeights_2:float=0.1851
    ConstraintWeights_3:float=0.1623
    ConstraintWeights_4:float=0.0211
    ConstraintWeights_5:float=0.0076
    ConstraintWeights_6:float=0.0029
   ) -> (Pose) {
    boneToModify {boneName=base_hair_06}
    constraintSetup [
     {targetBone {boneName=neck_02}}
     {targetBone {boneName=neck_01}}
     {targetBone {boneName=head}}
     {targetBone {boneName=spine_05}}
     {targetBone {boneName=spine_03}}
     {targetBone {boneName=spine_04}}
     {targetBone {boneName=dyn_hair_r_adjust}}
    ]
    constraintWeights [1,1,1,1,1,1,1]
   }
   n6 = "Constraint - base_hair_07 (PPPPPPPP)"!(
    ComponentPose=@n5
    Alpha=1
    ConstraintWeights_0=0.3846
    ConstraintWeights_1:float=0.2071
    ConstraintWeights_2:float=0.1695
    ConstraintWeights_3:float=0.1361
    ConstraintWeights_4:float=0.0857
    ConstraintWeights_5:float=0.0091
    ConstraintWeights_6:float=0.0041
    ConstraintWeights_7:float=0.0039
   ) -> (Pose) {
    boneToModify {boneName=base_hair_07}
    constraintSetup [
     {targetBone {boneName=head}}
     {targetBone {boneName=dyn_hair_07}}
     {targetBone {boneName=neck_02}}
     {targetBone {boneName=neck_01}}
     {targetBone {boneName=spine_05}}
     {targetBone {boneName=spine_03}}
     {targetBone {boneName=spine_04}}
     {targetBone {boneName=dyn_hair_r_adjust}}
    ]
    constraintWeights [1,1,1,1,1,1,1,1]
   }
   n7 = "Constraint - base_hair_08 (PP)"!(ComponentPose=@n6, Alpha=1, ConstraintWeights_0=0.7032, ConstraintWeights_1:float=0.2968) -> (Pose) {
    boneToModify {boneName=base_hair_08}
    constraintSetup [{targetBone {boneName=head}}, {targetBone {boneName=dyn_hair_08}}]
    constraintWeights [1,1]
   }
   n8 = "Constraint - base_hair_09 (PPPPPPPP)"!(
    ComponentPose=@n7
    Alpha=1
    ConstraintWeights_0=0.3366
    ConstraintWeights_1:float=0.2597
    ConstraintWeights_2:float=0.2358
    ConstraintWeights_3:float=0.0973
    ConstraintWeights_4:float=0.0373
    ConstraintWeights_5:float=0.0127
    ConstraintWeights_6:float=0.0117
    ConstraintWeights_7:float=0.009
   ) -> (Pose) {
    boneToModify {boneName=base_hair_09}
    constraintSetup [
     {targetBone {boneName=dyn_hair_09}}
     {targetBone {boneName=neck_01}}
     {targetBone {boneName=neck_02}}
     {targetBone {boneName=dyn_hair_r_adjust}}
     {targetBone {boneName=spine_03}}
     {targetBone {boneName=spine_05}}
     {targetBone {boneName=spine_04}}
     {targetBone {boneName=head}}
    ]
    constraintWeights [1,1,1,1,1,1,1,1]
   }
   n9 = "Constraint - base_hair_10 (PPPPPPP)"!(
    ComponentPose=@n8
    Alpha=1
    ConstraintWeights_0=0.3612
    ConstraintWeights_1:float=0.2229
    ConstraintWeights_2:float=0.1527
    ConstraintWeights_3:float=0.0921
    ConstraintWeights_4:float=0.0872
    ConstraintWeights_5:float=0.0663
    ConstraintWeights_6:float=0.0177
   ) -> (Pose) {
    boneToModify {boneName=base_hair_10}
    constraintSetup [
     {targetBone {boneName=spine_05}}
     {targetBone {boneName=neck_01}}
     {targetBone {boneName=neck_02}}
     {targetBone {boneName=dyn_hair_10}}
     {targetBone {boneName=dyn_hair_r_adjust}}
     {targetBone {boneName=spine_03}}
     {targetBone {boneName=spine_04}}
    ]
    constraintWeights [1,1,1,1,1,1,1]
   }
   n10 = "Constraint - base_hair_11 (PPPPPPP)"!(
    ComponentPose=@n9
    Alpha=1
    ConstraintWeights_0=0.2339
    ConstraintWeights_1:float=0.2189
    ConstraintWeights_2:float=0.2027
    ConstraintWeights_3:float=0.1484
    ConstraintWeights_4:float=0.1462
    ConstraintWeights_5:float=0.0299
    ConstraintWeights_6:float=0.0201
   ) -> (Pose) {
    boneToModify {boneName=base_hair_11}
    constraintSetup [
     {targetBone {boneName=dyn_hair_l_adjust}}
     {targetBone {boneName=dyn_hair_11}}
     {targetBone {boneName=neck_02}}
     {targetBone {boneName=spine_05}}
     {targetBone {boneName=neck_01}}
     {targetBone {boneName=spine_03}}
     {targetBone {boneName=spine_04}}
    ]
    constraintWeights [1,1,1,1,1,1,1]
   }
   n11 = "Constraint - base_hair_12 (PPPPPPPP)"!(
    ComponentPose=@n10
    Alpha=1
    ConstraintWeights_0=0.3549
    ConstraintWeights_1:float=0.3366
    ConstraintWeights_2:float=0.0968
    ConstraintWeights_3:float=0.0791
    ConstraintWeights_4:float=0.0735
    ConstraintWeights_5:float=0.0381
    ConstraintWeights_6:float=0.019
    ConstraintWeights_7:float=0.0019
   ) -> (Pose) {
    boneToModify {boneName=base_hair_12}
    constraintSetup [
     {targetBone {boneName=spine_05}}
     {targetBone {boneName=dyn_hair_12}}
     {targetBone {boneName=neck_01}}
     {targetBone {boneName=neck_02}}
     {targetBone {boneName=spine_03}}
     {targetBone {boneName=dyn_hair_l_adjust}}
     {targetBone {boneName=spine_04}}
     {targetBone {boneName=head}}
    ]
    constraintWeights [1,1,1,1,1,1,1,1]
   }
   n12 = "Constraint - base_hair_13 (PPPPPPPP)"!(
    ComponentPose=@n11
    Alpha=1
    ConstraintWeights_0=0.2462
    ConstraintWeights_1:float=0.2367
    ConstraintWeights_2:float=0.1858
    ConstraintWeights_3:float=0.1801
    ConstraintWeights_4:float=0.0743
    ConstraintWeights_5:float=0.059
    ConstraintWeights_6:float=0.0169
    ConstraintWeights_7:float=0.001
   ) -> (Pose) {
    boneToModify {boneName=base_hair_13}
    constraintSetup [
     {targetBone {boneName=spine_05}}
     {targetBone {boneName=neck_01}}
     {targetBone {boneName=dyn_hair_13}}
     {targetBone {boneName=neck_02}}
     {targetBone {boneName=dyn_hair_l_adjust}}
     {targetBone {boneName=spine_03}}
     {targetBone {boneName=spine_04}}
     {targetBone {boneName=clavicle_l}}
    ]
    constraintWeights [1,1,1,1,1,1,1,1]
   }
   "Constraint - base_hair_14 (PPPPPPP)"!(
    ComponentPose=@n12
    Alpha=1
    ConstraintWeights_0=0.2783
    ConstraintWeights_1:float=0.1933
    ConstraintWeights_2:float=0.1848
    ConstraintWeights_3:float=0.1581
    ConstraintWeights_4:float=0.1224
    ConstraintWeights_5:float=0.0576
    ConstraintWeights_6:float=0.0055
   ) -> (Pose) {
    boneToModify {boneName=base_hair_14}
    constraintSetup [
     {targetBone {boneName=spine_05}}
     {targetBone {boneName=dyn_hair_l_adjust}}
     {targetBone {boneName=neck_01}}
     {targetBone {boneName=dyn_hair_14}}
     {targetBone {boneName=neck_02}}
     {targetBone {boneName=spine_03}}
     {targetBone {boneName=spine_04}}
    ]
    constraintWeights [1,1,1,1,1,1,1]
   }
   n13 = Control Rig(Source=Input Pose()) {controlRigAssetReference {blueprintRigClass=&/M_MED_RadRat_Head_CtrlRig.M_MED_RadRat_Head_CtrlRig_C}}
   n14 = "Constraint - base_hair_04 (PPPPPP)"!(
    ComponentPose=Local To Component(LocalPose=@n13)
    Alpha=1
    ConstraintWeights_0=0.3557
    ConstraintWeights_1:float=0.2523
    ConstraintWeights_2:float=0.2012
    ConstraintWeights_3:float=0.1528
    ConstraintWeights_4:float=0.0277
    ConstraintWeights_5:float=0.0102
   ) -> (Pose) {
    boneToModify {boneName=base_hair_04}
    constraintSetup [
     {targetBone {boneName=neck_02}}
     {targetBone {boneName=neck_01}}
     {targetBone {boneName=spine_05}}
     {targetBone {boneName=head}}
     {targetBone {boneName=spine_03}}
     {targetBone {boneName=spine_04}}
    ]
    constraintWeights [1,1,1,1,1,1]
   }
   n15 = "Constraint - base_hair_05 (PPPPPPP)"!(
    ComponentPose=@n14
    Alpha=1
    ConstraintWeights_0=0.3317
    ConstraintWeights_1:float=0.278
    ConstraintWeights_2:float=0.2047
    ConstraintWeights_3:float=0.1488
    ConstraintWeights_4:float=0.0253
    ConstraintWeights_5:float=0.0092
    ConstraintWeights_6:float=0.0022
   ) -> (Pose) {
    boneToModify {boneName=base_hair_05}
    constraintSetup [
     {targetBone {boneName=neck_02}}
     {targetBone {boneName=neck_01}}
     {targetBone {boneName=spine_05}}
     {targetBone {boneName=head}}
     {targetBone {boneName=spine_03}}
     {targetBone {boneName=spine_04}}
     {targetBone {boneName=clavicle_r}}
    ]
    constraintWeights [1,1,1,1,1,1,1]
   }
   n16 = "Constraint - base_hair_08 (PPPPPP)"!(
    ComponentPose=@n15
    Alpha=1
    ConstraintWeights_0=0.7843
    ConstraintWeights_1:float=0.1544
    ConstraintWeights_2:float=0.0553
    ConstraintWeights_3:float=0.004
    ConstraintWeights_4:float=0.0011
    ConstraintWeights_5:float=0.0009
   ) -> (Pose) {
    boneToModify {boneName=base_hair_08}
    constraintSetup [
     {targetBone {boneName=head}}
     {targetBone {boneName=neck_02}}
     {targetBone {boneName=neck_01}}
     {targetBone {boneName=spine_05}}
     {targetBone {boneName=spine_03}}
     {targetBone {boneName=spine_04}}
    ]
    constraintWeights [1,1,1,1,1,1]
   }
   n17 = "Constraint - base_hair_07 (PPPPPP)"!(
    ComponentPose=@n16
    Alpha=1
    ConstraintWeights_0=0.401
    ConstraintWeights_1:float=0.334
    ConstraintWeights_2:float=0.1725
    ConstraintWeights_3:float=0.0794
    ConstraintWeights_4:float=0.0092
    ConstraintWeights_5:float=0.0038
   ) -> (Pose) {
    boneToModify {boneName=base_hair_07}
    constraintSetup [
     {targetBone {boneName=head}}
     {targetBone {boneName=neck_02}}
     {targetBone {boneName=neck_01}}
     {targetBone {boneName=spine_05}}
     {targetBone {boneName=spine_03}}
     {targetBone {boneName=spine_04}}
    ]
    constraintWeights [1,1,1,1,1,1]
   }
   n18 = "Constraint - base_hair_01 (PPPPPP)"!(
    ComponentPose=@n17
    Alpha=1
    ConstraintWeights_0=0.4943
    ConstraintWeights_1:float=0.2885
    ConstraintWeights_2:float=0.167
    ConstraintWeights_3:float=0.0447
    ConstraintWeights_4:float=0.0038
    ConstraintWeights_5:float=0.0016
   ) -> (Pose) {
    boneToModify {boneName=base_hair_01}
    constraintSetup [
     {targetBone {boneName=head}}
     {targetBone {boneName=neck_02}}
     {targetBone {boneName=neck_01}}
     {targetBone {boneName=spine_05}}
     {targetBone {boneName=spine_03}}
     {targetBone {boneName=spine_04}}
    ]
    constraintWeights [1,1,1,1,1,1]
   }
   n19 = "Constraint - base_hair_02 (PPPPPP)"!(
    ComponentPose=@n18
    Alpha=1
    ConstraintWeights_0=0.9513
    ConstraintWeights_1:float=0.0349
    ConstraintWeights_2:float=0.0123
    ConstraintWeights_3:float=0.001
    ConstraintWeights_4:float=0.0002
    ConstraintWeights_5:float=0.0002
   ) -> (Pose) {
    boneToModify {boneName=base_hair_02}
    constraintSetup [
     {targetBone {boneName=head}}
     {targetBone {boneName=neck_02}}
     {targetBone {boneName=neck_01}}
     {targetBone {boneName=spine_05}}
     {targetBone {boneName=spine_03}}
     {targetBone {boneName=spine_04}}
    ]
    constraintWeights [1,1,1,1,1,1]
   }
   n20 = Anim Dynamics - Bone: dyn_hair_01!(SimulationSpace:AnimPhysSimSpaceType Enum, ComponentPose=@n19, Alpha=1, ExternalForce) -> (Pose) {
    angularDampingOverride=1
    angularSpringConstant=10000
    bAngularSpring=true
    bOverrideAngularDamping=true
    bUseGravityOverride=true
    gravityScale=0.35
    simulationSpace=RootRelative
    boundBone {boneName=dyn_hair_01}
    componentAppliedLinearAccClamp=<2000 2000 2000>
    componentLinearAccScale=<1 1 1>
    componentLinearVelScale=<1 1 1>
    gravityOverride {z=-980}
    physicsBodyDefinitions [{
      boundBone {boneName=dyn_hair_01}
      boxExtents=<20 30 30>
      constraintSetup {angularLimitsMax=<30 30 30>, angularLimitsMin=<-30 -30 -30>, angularTarget=<1 0 0>}
      localJointOffset=<5 0 0>
     }]
   }
   n21 = Anim Dynamics - Bone: dyn_hair_02!(SimulationSpace:AnimPhysSimSpaceType Enum, ComponentPose=@n20, Alpha=1, ExternalForce) -> (Pose) {
    angularDampingOverride=1
    angularSpringConstant=10000
    bAngularSpring=true
    bOverrideAngularDamping=true
    bUseGravityOverride=true
    gravityScale=0.35
    simulationSpace=RootRelative
    boundBone {boneName=dyn_hair_02}
    componentAppliedLinearAccClamp=<2000 2000 2000>
    componentLinearAccScale=<1 1 1>
    componentLinearVelScale=<1 1 1>
    gravityOverride {z=-980}
    physicsBodyDefinitions [{
      boundBone {boneName=dyn_hair_02}
      boxExtents=<20 30 30>
      constraintSetup {angularLimitsMax=<30 30 30>, angularLimitsMin=<-30 -30 -30>, angularTarget=<1 0 0>}
      localJointOffset=<5 0 0>
     }]
   }
   n22 = Anim Dynamics - Bone: dyn_hair_03!(SimulationSpace:AnimPhysSimSpaceType Enum, ComponentPose=@n21, Alpha=1, ExternalForce="0,0,0") -> (Pose) {
    angularDampingOverride=1
    angularSpringConstant=10000
    bAngularSpring=true
    bOverrideAngularDamping=true
    bUseGravityOverride=true
    gravityScale=0.35
    simulationSpace=RootRelative
    boundBone {boneName=dyn_hair_03}
    componentAppliedLinearAccClamp=<2000 2000 2000>
    componentLinearAccScale=<1 1 1>
    componentLinearVelScale=<1 1 1>
    gravityOverride {z=-980}
    physicsBodyDefinitions [{
      boundBone {boneName=dyn_hair_03}
      boxExtents=<20 30 30>
      constraintSetup {angularLimitsMax=<30 30 30>, angularLimitsMin=<-30 -30 -30>, angularTarget=<1 0 0>}
      localJointOffset=<5 0 0>
     }]
   }
   n23 = Anim Dynamics - Bone: dyn_hair_04!(SimulationSpace:AnimPhysSimSpaceType Enum, ComponentPose=@n22, Alpha=1, ExternalForce="0,0,0") -> (Pose) {
    angularDampingOverride=1
    angularSpringConstant=10000
    bAngularSpring=true
    bOverrideAngularDamping=true
    bUseGravityOverride=true
    gravityScale=0.35
    simulationSpace=RootRelative
    boundBone {boneName=dyn_hair_04}
    componentAppliedLinearAccClamp=<2000 2000 2000>
    componentLinearAccScale=<1 1 1>
    componentLinearVelScale=<1 1 1>
    gravityOverride {z=-980}
    physicsBodyDefinitions [{
      boundBone {boneName=dyn_hair_04}
      boxExtents=<20 30 30>
      constraintSetup {angularLimitsMax=<30 30 30>, angularLimitsMin=<-30 -30 -30>, angularTarget=<1 0 0>}
      localJointOffset=<5 0 0>
     }]
   }
   n24 = Anim Dynamics - Bone: dyn_hair_05!(SimulationSpace:AnimPhysSimSpaceType Enum, ComponentPose=@n23, Alpha=1, ExternalForce="0,0,0") -> (Pose) {
    angularDampingOverride=1
    angularSpringConstant=10000
    bAngularSpring=true
    bOverrideAngularDamping=true
    bUseGravityOverride=true
    gravityScale=0.35
    simulationSpace=RootRelative
    boundBone {boneName=dyn_hair_05}
    componentAppliedLinearAccClamp=<2000 2000 2000>
    componentLinearAccScale=<1 1 1>
    componentLinearVelScale=<1 1 1>
    gravityOverride {z=-980}
    physicsBodyDefinitions [{
      boundBone {boneName=dyn_hair_05}
      boxExtents=<20 30 30>
      constraintSetup {angularLimitsMax=<30 30 30>, angularLimitsMin=<-30 -30 -30>, angularTarget=<1 0 0>}
      localJointOffset=<5 0 0>
     }]
   }
   n25 = Anim Dynamics - Bone: dyn_hair_06!(SimulationSpace:AnimPhysSimSpaceType Enum, ComponentPose=@n24, Alpha=1, ExternalForce="0,0,0") -> (Pose) {
    angularDampingOverride=1
    angularSpringConstant=10000
    bAngularSpring=true
    bOverrideAngularDamping=true
    bUseGravityOverride=true
    gravityScale=0.35
    simulationSpace=RootRelative
    boundBone {boneName=dyn_hair_06}
    componentAppliedLinearAccClamp=<2000 2000 2000>
    componentLinearAccScale=<1 1 1>
    componentLinearVelScale=<1 1 1>
    gravityOverride {z=-980}
    physicsBodyDefinitions [{
      boundBone {boneName=dyn_hair_06}
      boxExtents=<20 30 30>
      constraintSetup {angularLimitsMax=<30 30 30>, angularLimitsMin=<-30 -30 -30>, angularTarget=<1 0 0>}
      localJointOffset=<5 0 0>
     }]
   }
   n26 = Anim Dynamics - Bone: dyn_hair_07!(SimulationSpace:AnimPhysSimSpaceType Enum, ComponentPose=@n25, Alpha=1, ExternalForce="0,0,0") -> (Pose) {
    angularDampingOverride=1
    angularSpringConstant=10000
    bAngularSpring=true
    bOverrideAngularDamping=true
    bUseGravityOverride=true
    gravityScale=0.35
    simulationSpace=RootRelative
    boundBone {boneName=dyn_hair_07}
    componentAppliedLinearAccClamp=<2000 2000 2000>
    componentLinearAccScale=<1 1 1>
    componentLinearVelScale=<1 1 1>
    gravityOverride {z=-980}
    physicsBodyDefinitions [{
      boundBone {boneName=dyn_hair_07}
      boxExtents=<20 30 30>
      constraintSetup {angularLimitsMax=<30 30 30>, angularLimitsMin=<-30 -30 -30>, angularTarget=<1 0 0>}
      localJointOffset=<5 0 0>
     }]
   }
   n27 = Anim Dynamics - Bone: dyn_hair_08!(SimulationSpace:AnimPhysSimSpaceType Enum, ComponentPose=@n26, Alpha=1, ExternalForce="0,0,0") -> (Pose) {
    angularDampingOverride=1
    angularSpringConstant=10000
    bAngularSpring=true
    bOverrideAngularDamping=true
    bUseGravityOverride=true
    gravityScale=0.35
    simulationSpace=RootRelative
    boundBone {boneName=dyn_hair_08}
    componentAppliedLinearAccClamp=<2000 2000 2000>
    componentLinearAccScale=<1 1 1>
    componentLinearVelScale=<1 1 1>
    gravityOverride {z=-980}
    physicsBodyDefinitions [{
      boundBone {boneName=dyn_hair_08}
      boxExtents=<20 30 30>
      constraintSetup {angularLimitsMax=<30 30 30>, angularLimitsMin=<-30 -30 -30>, angularTarget=<1 0 0>}
      localJointOffset=<5 0 0>
     }]
   }
   n28 = Anim Dynamics - Bone: dyn_hair_09!(SimulationSpace:AnimPhysSimSpaceType Enum, ComponentPose=@n27, Alpha=1, ExternalForce="0,0,0") -> (Pose) {
    angularDampingOverride=1
    angularSpringConstant=10000
    bAngularSpring=true
    bOverrideAngularDamping=true
    bUseGravityOverride=true
    gravityScale=0.35
    simulationSpace=RootRelative
    boundBone {boneName=dyn_hair_09}
    componentAppliedLinearAccClamp=<2000 2000 2000>
    componentLinearAccScale=<1 1 1>
    componentLinearVelScale=<1 1 1>
    gravityOverride {z=-980}
    physicsBodyDefinitions [{
      boundBone {boneName=dyn_hair_09}
      boxExtents=<20 30 30>
      constraintSetup {angularLimitsMax=<30 30 30>, angularLimitsMin=<-30 -30 -30>, angularTarget=<1 0 0>}
      localJointOffset=<5 0 0>
     }]
   }
   n29 = Anim Dynamics - Bone: dyn_hair_10!(SimulationSpace:AnimPhysSimSpaceType Enum, ComponentPose=@n28, Alpha=1, ExternalForce="0,0,0") -> (Pose) {
    angularDampingOverride=1
    angularSpringConstant=10000
    bAngularSpring=true
    bOverrideAngularDamping=true
    bUseGravityOverride=true
    gravityScale=0.35
    simulationSpace=RootRelative
    boundBone {boneName=dyn_hair_10}
    componentAppliedLinearAccClamp=<2000 2000 2000>
    componentLinearAccScale=<1 1 1>
    componentLinearVelScale=<1 1 1>
    gravityOverride {z=-980}
    physicsBodyDefinitions [{
      boundBone {boneName=dyn_hair_10}
      boxExtents=<20 30 30>
      constraintSetup {angularLimitsMax=<30 30 30>, angularLimitsMin=<-30 -30 -30>, angularTarget=<1 0 0>}
      localJointOffset=<5 0 0>
     }]
   }
   n30 = Anim Dynamics - Bone: dyn_hair_11!(SimulationSpace:AnimPhysSimSpaceType Enum, ComponentPose=@n29, Alpha=1, ExternalForce="0,0,0") -> (Pose) {
    angularDampingOverride=1
    angularSpringConstant=10000
    bAngularSpring=true
    bOverrideAngularDamping=true
    bUseGravityOverride=true
    gravityScale=0.35
    simulationSpace=RootRelative
    boundBone {boneName=dyn_hair_11}
    componentAppliedLinearAccClamp=<2000 2000 2000>
    componentLinearAccScale=<1 1 1>
    componentLinearVelScale=<1 1 1>
    gravityOverride {z=-980}
    physicsBodyDefinitions [{
      boundBone {boneName=dyn_hair_11}
      boxExtents=<20 30 30>
      constraintSetup {angularLimitsMax=<30 30 30>, angularLimitsMin=<-30 -30 -30>, angularTarget=<1 0 0>}
      localJointOffset=<5 0 0>
     }]
   }
   n31 = Anim Dynamics - Bone: dyn_hair_12!(SimulationSpace:AnimPhysSimSpaceType Enum, ComponentPose=@n30, Alpha=1, ExternalForce="0,0,0") -> (Pose) {
    angularDampingOverride=1
    angularSpringConstant=10000
    bAngularSpring=true
    bOverrideAngularDamping=true
    bUseGravityOverride=true
    gravityScale=0.35
    simulationSpace=RootRelative
    boundBone {boneName=dyn_hair_12}
    componentAppliedLinearAccClamp=<2000 2000 2000>
    componentLinearAccScale=<1 1 1>
    componentLinearVelScale=<1 1 1>
    gravityOverride {z=-980}
    physicsBodyDefinitions [{
      boundBone {boneName=dyn_hair_12}
      boxExtents=<20 30 30>
      constraintSetup {angularLimitsMax=<30 30 30>, angularLimitsMin=<-30 -30 -30>, angularTarget=<1 0 0>}
      localJointOffset=<5 0 0>
     }]
   }
   n32 = Anim Dynamics - Bone: dyn_hair_13!(SimulationSpace:AnimPhysSimSpaceType Enum, ComponentPose=@n31, Alpha=1, ExternalForce="0,0,0") -> (Pose) {
    angularDampingOverride=1
    angularSpringConstant=10000
    bAngularSpring=true
    bOverrideAngularDamping=true
    bUseGravityOverride=true
    gravityScale=0.35
    simulationSpace=RootRelative
    boundBone {boneName=dyn_hair_13}
    componentAppliedLinearAccClamp=<2000 2000 2000>
    componentLinearAccScale=<1 1 1>
    componentLinearVelScale=<1 1 1>
    gravityOverride {z=-980}
    physicsBodyDefinitions [{
      boundBone {boneName=dyn_hair_13}
      boxExtents=<20 30 30>
      constraintSetup {angularLimitsMax=<30 30 30>, angularLimitsMin=<-30 -30 -30>, angularTarget=<1 0 0>}
      localJointOffset=<5 0 0>
     }]
   }
   n33 = Anim Dynamics - Bone: dyn_hair_14!(SimulationSpace:AnimPhysSimSpaceType Enum, ComponentPose=@n32, Alpha=1, ExternalForce="0,0,0") -> (Pose) {
    angularDampingOverride=1
    angularSpringConstant=10000
    bAngularSpring=true
    bOverrideAngularDamping=true
    bUseGravityOverride=true
    gravityScale=0.35
    simulationSpace=RootRelative
    boundBone {boneName=dyn_hair_14}
    componentAppliedLinearAccClamp=<2000 2000 2000>
    componentLinearAccScale=<1 1 1>
    componentLinearVelScale=<1 1 1>
    gravityOverride {z=-980}
    physicsBodyDefinitions [{
      boundBone {boneName=dyn_hair_14}
      boxExtents=<20 30 30>
      constraintSetup {angularLimitsMax=<30 30 30>, angularLimitsMin=<-30 -30 -30>, angularTarget=<1 0 0>}
      localJointOffset=<5 0 0>
     }]
   }
   Output Pose(Result=Component To Local(ComponentPose=@n33))
  ]}
 EventGraph {nodes [
   Event BlueprintUpdateAnimation()
   > n1 = Branch(Condition=Get bIsInFrontEnd())
   > Set Sim Space(Sim Space:AnimPhysSimSpaceType Enum=World) -> (Output_Get:AnimPhysSimSpaceType Enum)
   > Set DynAlpha(DynAlpha:double=0.65) -> (Output_Get:double)
   Set Sim Space(execute=@n1.else)
   > Set DynAlpha(DynAlpha=1)
   TryGetPawnOwner()
  ]}
]
variables [
 DynAlpha {type=number}
 Sim Space {type=AnimPhysSimSpaceType}
]
